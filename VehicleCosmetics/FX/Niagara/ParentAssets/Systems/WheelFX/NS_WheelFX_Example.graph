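# Example Niagara system for vehicle wheel cosmetics, with mesh, sprite, world-space and wheel-wrap emitters, plus a disabled skeletal-mesh variant.
# Shared VC_WheelFX setup modules partition particles by wheel, then SkeletalMeshLocation samples each wheel's bone or socket for VC_WheelXform.
# Attached emitters re-sample every update, while WorldSpace samples only at spawn and shortens particle lifetime with vehicle speed.
# WheelWrap scales meshes to wheel size and switches dynamic material parameters on User.bIsPainted.
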NiagaraSystem /VehicleCosmetics/FX/Niagara/ParentAssets/Systems/WheelFX/NS_WheelFX_Example
dependencies {
 usedDataInterfaces [
  &NiagaraDataInterfaceCurve
  &NiagaraDataInterfaceSpriteRendererInfo
  &NiagaraDataInterfaceSkeletalMesh
 ]
 usedDynamicInputs [&MakeCustomVector2DFromBool]
 usedModules [
  &self:SystemSpawnScript.SystemScriptSource.SystemScriptGraph.NiagaraNodeAssignment_0.SetVariables_#1
  &FN_UnpackVehicleFlags
  /VehicleCosmetics/FX/Niagara/Modules/WheelFX/VC_WheelFX_SystemSetup
  /VehicleCosmetics/FX/Niagara/Modules/VC_SetVehicleSpeedAttribs
  &SystemState
 ]
}
emitters {
 WheelFX_Mesh {
  emitterSpawnScript [/VehicleCosmetics/FX/Niagara/Modules/WheelFX/VC_WheelFX_EmitterSetup()]
  emitterUpdateScript [
   &EmitterState(Life Cycle Mode=::NewEnumerator1, Loop Behavior=::NewEnumerator1, Loop Duration Mode=::NewEnumerator1, UseLoopDelay=false)
   &SpawnBurst_Instantaneous(Spawn Count=$Emitter.NumWheels, Use Spawn Probability=false, UseLoopCountLimit=false)
   &SpawnRate(SpawnRate=15, Use Spawn Probability=false) {enabled=false}
  ]
  particleSpawnScript [
   &PartitionParticles(Number Of Partitions=$Emitter.NumWheels)
   /VehicleCosmetics/FX/Niagara/Modules/WheelFX/VC_WheelFX_ParticleSetup(WheelID:NiagaraInt32=$Particles.PartitionParticles.Partition)
   &VC_PaintedVariantsOverride() {enabled=false}
   &InitializeParticle(
    Color Mode=::NewEnumerator0
    Mesh Scale Mode=::NewEnumerator3
    Mesh Uniform Scale=2.5
    Position=<0 25 0>
    Position Mode=::NewEnumerator1
    Sprite Size Mode=::NewEnumerator3
    Uniform Sprite Size=20
    UsePositionOffset=false
    Write Material Random=true
   )
   /SetVariables(Particles.SpriteColor=<rgba 4.389 5 0.1131 0.2>) {isSetParameters=true}
   &ShapeLocation(
    MaskSpawnGroup=false
    Override Seed=false
    Ring Radius=&Multiply_Float(A=$Particles.WheelRadius, B=0.5)
    Rotation Angle=90
    Rotation Angle Type:ENiagara_AngleInput=::NewEnumerator0
    Rotation Axis=<0 1 0>
    Rotation Mode=::NewEnumerator1
    Shape Primitive=::NewEnumerator4
    UseOwnerScale=true
   ) {enabled=false}
   &AddVelocity(
    Cone Angle=30
    Cone Axis=<0 1 0>
    Module.UseConeFalloff=false
    Module.UseVelDistribution=false
    Override Seed=false
    Rotation Coordinate Space=::Simulation
    Velocity Mode=::NewEnumerator2
    Velocity Speed=150
   ) {enabled=false}
  ]
  particleUpdateScript [
   &ParticleState(Kill Particles When Lifetime Has Elapsed=false, Let Infinitely Lived Particles Die When Emitter Deactivates=true)
   /SetVariables(Particles.MeshOrientation=/CascadeToNiagaraConverter/NiagaraScripts/DynamicInputScripts/MakeQuatFromEuler(
     Euler:Vector3f=&MakeVector(
      X=&Waveform(
       Global Amplitude Scale=&MakeCustomFloatFromBool(Bool=$Particles.WheelIsFront, True Float=0, False Float=45)
       "Waveform [1]"=::NewEnumerator0
       "[1] Frequency"=1
       "[1] Amplitude Min/Max"=<-1 1>
      )
      Y=&Waveform(Global Amplitude Scale=0, "Waveform [1]"=::NewEnumerator0, "[1] Frequency"=1, "[1] Amplitude Min/Max"=<-1 1>)
      Z=&Waveform(
       Global Amplitude Scale=&MakeCustomFloatFromBool(Bool=$Particles.WheelIsFront, True Float=45, False Float=0)
       "Waveform [1]"=::NewEnumerator0
       "[1] Frequency"=1
       "[1] Amplitude Min/Max"=<-1 1>
      )
     )
     Scale:NiagaraFloat=1
    )) {isSetParameters=true}
   &SolveForcesAndVelocity(Clamp Velocity=false, Limit Acceleration=false)
   &SkeletalMeshLocation(
    Bone / Socket Index=$Particles.WheelID
    Bone / Socket Sampling Mode=::NewEnumerator5
    MaskSpawnGroup=false
    Mesh Sampling Type=::NewEnumerator6
    Orientation Sampling=::NewEnumerator1
    Position Sampling=::NewEnumerator2
    Skeletal Mesh=$System.SkeletalMesh
   )
   /VehicleCosmetics/FX/Niagara/Modules/WheelFX/VC_WheelXform(
    SKM Position:NiagaraPosition=$Particles.SkeletalMeshLocation.SampledPosition
    SKM Rotation:Quat4f=$Particles.SkeletalMeshLocation.SampledRotation
    SKM Velocity:Vector3f=$Particles.SkeletalMeshLocation.SampledVelocity
   )
  ]
  rendererClasses [&NiagaraMeshRendererProperties, &NiagaraSpriteRendererProperties]
  renderers [
   {rendererClass=&NiagaraMeshRendererProperties, rendererIndex=0}
   {rendererClass=&NiagaraSpriteRendererProperties, rendererIndex=1}
  ]
 }
 WheelFX_SkeletalMesh {
  bEnabled=false
  emitterSpawnScript [&VC_WheelFX_EmitterSetup()]
  emitterUpdateScript [
   &EmitterState(Life Cycle Mode=::NewEnumerator1, Loop Behavior=::NewEnumerator1, Loop Duration Mode=::NewEnumerator1, UseLoopDelay=false)
   &SpawnBurst_Instantaneous(Spawn Count=$Emitter.NumWheels, Use Spawn Probability=false, UseLoopCountLimit=false)
   &SpawnRate(SpawnRate=15, Use Spawn Probability=false) {enabled=false}
  ]
  particleSpawnScript [
   &PartitionParticles(Number Of Partitions=$Emitter.NumWheels)
   &VC_WheelFX_ParticleSetup(WheelID:NiagaraInt32=$Particles.PartitionParticles.Partition)
   &VC_PaintedVariantsOverride()
   &InitializeParticle(
    Color Mode=::NewEnumerator0
    Mesh Scale Mode=::NewEnumerator3
    Mesh Uniform Scale=0.5
    Position=<-100 25 0>
    Position Mode=::NewEnumerator1
    UsePositionOffset=false
    Write Material Random=true
   )
   &ShapeLocation(
    MaskSpawnGroup=false
    Override Seed=false
    Ring Radius=&Multiply_Float(A=$Particles.WheelRadius, B=0.5)
    Rotation Angle=90
    Rotation Angle Type:ENiagara_AngleInput=::NewEnumerator0
    Rotation Axis=<0 1 0>
    Rotation Mode=::NewEnumerator1
    Shape Primitive=::NewEnumerator4
    UseOwnerScale=true
   ) {enabled=false}
   &AddVelocity(
    Cone Angle=30
    Cone Axis=<0 1 0>
    Module.UseConeFalloff=false
    Module.UseVelDistribution=false
    Override Seed=false
    Rotation Coordinate Space=::Simulation
    Velocity Mode=::NewEnumerator2
    Velocity Speed=150
   ) {enabled=false}
  ]
  particleUpdateScript [
   &ParticleState(Kill Particles When Lifetime Has Elapsed=false, Let Infinitely Lived Particles Die When Emitter Deactivates=true)
   &SolveForcesAndVelocity(Clamp Velocity=false, Limit Acceleration=false)
   &SkeletalMeshLocation(
    Bone / Socket Index=$Particles.WheelID
    Bone / Socket Sampling Mode=::NewEnumerator5
    MaskSpawnGroup=false
    Mesh Sampling Type=::NewEnumerator6
    Orientation Sampling=::NewEnumerator1
    Position Sampling=::NewEnumerator2
    Skeletal Mesh=$System.SkeletalMesh
   )
   &VC_WheelXform(
    SKM Position:NiagaraPosition=$Particles.SkeletalMeshLocation.SampledPosition
    SKM Rotation:Quat4f=$Particles.SkeletalMeshLocation.SampledRotation
    SKM Velocity:Vector3f=$Particles.SkeletalMeshLocation.SampledVelocity
   )
   /SetVariables(Particles.VC_WheelXform.Position:NiagaraPosition=&AddVectorToPosition(
     Position=$Particles.VC_WheelXform.Position
     Vector=/VehicleCosmetics/FX/Niagara/Modules/WheelFX/VC_ConvertWheelVector(
      " Position Offset":Vector3f=&Multiply_Vector(A=$Particles.Scale, B=<-280 200 0>)
      Wheel Is Left Side:NiagaraBool=$Particles.WheelIsLeftSide
      MirrorLeftSideOrient:NiagaraBool=$Emitter.MirrorLeftSideOrient
     )
    )) {enabled=false, isSetParameters=true}
  ]
  rendererClasses [
   &NiagaraComponentRendererProperties
   &NiagaraMeshRendererProperties
   &NiagaraSpriteRendererProperties
  ]
  renderers [
   {rendererClass=&NiagaraComponentRendererProperties, rendererIndex=0}
   {rendererClass=&NiagaraMeshRendererProperties, rendererIndex=1}
   {rendererClass=&NiagaraSpriteRendererProperties, rendererIndex=2}
  ]
 }
 WheelFX_Sprites {
  emitterSpawnScript [&VC_WheelFX_EmitterSetup()]
  emitterUpdateScript [
   &EmitterState(Life Cycle Mode=::NewEnumerator1, Loop Behavior=::NewEnumerator1, Loop Duration Mode=::NewEnumerator1, UseLoopDelay=false)
   &SpawnRate(SpawnRate=15, Use Spawn Probability=false) {module=SpawnRate001}
   &SpawnBurst_Instantaneous(Spawn Count=$Emitter.NumWheels, Use Spawn Probability=false, UseLoopCountLimit=false) {enabled=false}
   &SpawnRate(SpawnRate=15, Use Spawn Probability=false) {enabled=false}
  ]
  particleSpawnScript [
   &PartitionParticles(Number Of Partitions=$Emitter.NumWheels, Partition Mode:ENiagara_PartitionMode_v2=::NewEnumerator5)
   &VC_WheelFX_ParticleSetup(WheelID:NiagaraInt32=$Particles.PartitionParticles.Partition)
   &VC_PaintedVariantsOverride()
   &InitializeParticle(
    Color Mode=::NewEnumerator0
    Position Mode=::NewEnumerator1
    Sprite Size Mode=::NewEnumerator3
    Uniform Sprite Size=50
    UsePositionOffset=false
    Write Material Random=true
   )
   &ShapeLocation(
    MaskSpawnGroup=false
    Override Seed=false
    Ring Radius=&Multiply_Float(A=$Particles.WheelRadius, B=0.35)
    Rotation Angle=90
    Rotation Angle Type:ENiagara_AngleInput=::NewEnumerator0
    Rotation Axis=<0 1 0>
    Rotation Mode=::NewEnumerator1
    Shape Primitive=::NewEnumerator4
    UseOwnerScale=true
   )
   &AddVelocity(
    Cone Angle=30
    Cone Axis=<0 1 0>
    Module.UseConeFalloff=false
    Module.UseVelDistribution=false
    Override Seed=false
    Rotation Coordinate Space=::Simulation
    Velocity Mode=::NewEnumerator2
    Velocity Speed=76
   )
  ]
  particleUpdateScript [
   &ParticleState()
   &ScaleColor(
    Scale Alpha=&FloatFromCurve(FloatCurve=DataInterface() {propertyValues {Curve {keys [<at 0.75 1>,<at 1 0>]}}})
    ScaleA=true
    ScaleRGB=false
   )
   &SubUVAnimation(
    "Sprite Renderer "=DataInterface() {propertyValues {SpriteRenderer=&self:WheelFX_Sprites_0.NiagaraSpriteRendererProperties_1}}
    UseEndFrame=false
    UseStartFrame=false
   )
   &SolveForcesAndVelocity(Clamp Velocity=false, Limit Acceleration=false)
   &SkeletalMeshLocation(
    Bone / Socket Index=$Particles.WheelID
    Bone / Socket Sampling Mode=::NewEnumerator5
    MaskSpawnGroup=false
    Mesh Sampling Type=::NewEnumerator6
    Orientation Sampling=::NewEnumerator1
    Position Sampling=::NewEnumerator2
    Skeletal Mesh=$System.SkeletalMesh
   )
   &VC_WheelXform(
    SKM Position:NiagaraPosition=$Particles.SkeletalMeshLocation.SampledPosition
    SKM Rotation:Quat4f=$Particles.SkeletalMeshLocation.SampledRotation
    SKM Velocity:Vector3f=$Particles.SkeletalMeshLocation.SampledVelocity
   )
  ]
  rendererClasses [&NiagaraMeshRendererProperties, &NiagaraSpriteRendererProperties]
  renderers [
   {rendererClass=&NiagaraMeshRendererProperties, rendererIndex=0}
   {rendererClass=&NiagaraSpriteRendererProperties, rendererIndex=1}
  ]
 }
 WheelFX_WheelWrap {
  emitterSpawnScript [&VC_WheelFX_EmitterSetup()]
  emitterUpdateScript [
   &EmitterState(Life Cycle Mode=::NewEnumerator1, Loop Behavior=::NewEnumerator1, Loop Duration Mode=::NewEnumerator1, UseLoopDelay=false)
   &SpawnBurst_Instantaneous(Spawn Count=$Emitter.NumWheels, Use Spawn Probability=false, UseLoopCountLimit=false)
   &SpawnRate(SpawnRate=15, Use Spawn Probability=false) {enabled=false}
  ]
  particleSpawnScript [
   &PartitionParticles(Number Of Partitions=$Emitter.NumWheels)
   &VC_WheelFX_ParticleSetup(WheelID:NiagaraInt32=$Particles.PartitionParticles.Partition)
   &VC_PaintedVariantsOverride()
   &InitializeParticle(
    Color Mode=::NewEnumerator0
    Mesh Scale=$Particles.WheelScale
    Mesh Scale Mode=::NewEnumerator4
    Position Mode=::NewEnumerator1
    UsePositionOffset=false
    Write Material Random=true
   )
   &ShapeLocation(
    MaskSpawnGroup=false
    Override Seed=false
    Ring Radius=&Multiply_Float(A=$Particles.WheelRadius, B=0.5)
    Rotation Angle=90
    Rotation Angle Type:ENiagara_AngleInput=::NewEnumerator0
    Rotation Axis=<0 1 0>
    Rotation Mode=::NewEnumerator1
    Shape Primitive=::NewEnumerator4
    UseOwnerScale=true
   ) {enabled=false}
   &AddVelocity(
    Cone Angle=30
    Cone Axis=<0 1 0>
    Module.UseConeFalloff=false
    Module.UseVelDistribution=false
    Override Seed=false
    Rotation Coordinate Space=::Simulation
    Velocity Mode=::NewEnumerator2
    Velocity Speed=150
   ) {enabled=false}
   &InitialMeshOrientation(Mesh Orientation Mode=::NewEnumerator2, Orientation Vector=<-1 0 0>, Use Rotation Vector=false) {enabled=false}
  ]
  particleUpdateScript [
   &ParticleState(Kill Particles When Lifetime Has Elapsed=false, Let Infinitely Lived Particles Die When Emitter Deactivates=true)
   &SolveForcesAndVelocity(Clamp Velocity=false, Limit Acceleration=false)
   &SkeletalMeshLocation(
    Bone / Socket Index=$Particles.WheelID
    Bone / Socket Sampling Mode=::NewEnumerator5
    MaskSpawnGroup=false
    Mesh Sampling Type=::NewEnumerator6
    Orientation Sampling=::NewEnumerator1
    Position Sampling=::NewEnumerator2
    Skeletal Mesh=$System.SkeletalMesh
   )
   &VC_WheelXform(
    SKM Position:NiagaraPosition=$Particles.SkeletalMeshLocation.SampledPosition
    SKM Rotation:Quat4f=$Particles.SkeletalMeshLocation.SampledRotation
    SKM Velocity:Vector3f=$Particles.SkeletalMeshLocation.SampledVelocity
    WrapStyleRotation=true
   )
   &DynamicMaterialParameters(
    Index 0 Param 1=&MakeCustomFloatFromBool(Bool=$User.bIsPainted, True Float=1, False Float=0)
    Index 0 Param 2=&MakeCustomFloatFromBool(Bool=$User.bIsPainted, True Float=0, False Float=2.9)
    Index 0 Param 3=&SelectFloatFromArray(Direct Array Index=&SpawnGroup(), Float Selection Array=DataInterface() {propertyValues {FloatData [0.3,0]}})
    Index 1 Linear Color:LinearColor=&MakeCustomLinearColorFromBool(Bool=$User.bIsPainted, True Color=$Particles.Color, False Color=<rgba 0.01 0.34 1 1>)
    Index 2 Linear Color:LinearColor=&MakeCustomLinearColorFromBool(Bool=$User.bIsPainted, True Color=$Particles.Color, False Color=<rgba 1.2 0.18 0.945 1>)
    Module.Index0Float0Write=true
    Module.Index0Float1Write=true
    Module.Index0Float2Write=true
    Module.Index0Float3Write=true
    Write Parameter Index 1=::NewEnumerator5
    Write Parameter Index 2=::NewEnumerator5
   )
  ]
  rendererClasses [&NiagaraMeshRendererProperties, &NiagaraSpriteRendererProperties]
  renderers [
   {rendererClass=&NiagaraMeshRendererProperties, rendererIndex=0}
   {rendererClass=&NiagaraMeshRendererProperties, rendererIndex=1}
   {rendererClass=&NiagaraSpriteRendererProperties, rendererIndex=2}
  ]
 }
 WheelFX_WorldSpace {
  emitterSpawnScript [&VC_WheelFX_EmitterSetup()]
  emitterUpdateScript [
   &EmitterState(Life Cycle Mode=::NewEnumerator1, Loop Behavior=::NewEnumerator1, Loop Duration Mode=::NewEnumerator1, UseLoopDelay=false)
   &SpawnBurst_Instantaneous(Spawn Count=$Emitter.NumWheels, Use Spawn Probability=false, UseLoopCountLimit=false) {enabled=false}
   &SpawnRate(SpawnRate=30, Use Spawn Probability=false)
  ]
  particleSpawnScript [
   &PartitionParticles(Number Of Partitions=$Emitter.NumWheels)
   &VC_WheelFX_ParticleSetup(WheelID:NiagaraInt32=$Particles.PartitionParticles.Partition)
   &VC_PaintedVariantsOverride()
   &InitializeParticle(
    Color=<rgba 1 0.2773 0.002 1>
    Lifetime=&Multiply_Float(
     A=1
     B=/DealershipSharedLibrary/FX/Niagara/DynamicInputs/DLR_RemapFloatBySpeed(
      Value at Min Speed:NiagaraFloat=1
      Value at Max Speed:NiagaraFloat=0.5
      Min Speed:NiagaraFloat=200
      Max Speed:NiagaraFloat=500
      Input Velocity:Vector3f=$Engine.Owner.Velocity
      Input Coordinate Space:ENiagaraCoordinateSpace=::World
      Local Velocity Mask:Vector3f=<1 1 1>
     )
    )
    Position Mode=::NewEnumerator1
    Sprite Rotation Angle=90
    Sprite Rotation Mode=::NewEnumerator0
    Sprite Size Mode=::NewEnumerator3
    UsePositionOffset=false
    Write Material Random=true
   )
   &SubUVAnimation(
    "Sprite Renderer "=DataInterface() {propertyValues {SpriteRenderer=&self:WheelFX_WorldSpace_1.NiagaraSpriteRendererProperties_5}}
    SubUV Animation Mode:ENiagara_SubUVLookupModeV2=::NewEnumerator2
    UseEndFrame=false
    UseStartFrame=false
   )
   &ShapeLocation(
    MaskSpawnGroup=false
    Override Seed=false
    Ring Radius=&Multiply_Float(A=$Particles.WheelRadius, B=0.5)
    Rotation Angle=90
    Rotation Angle Type:ENiagara_AngleInput=::NewEnumerator0
    Rotation Axis=<0 1 0>
    Rotation Mode=::NewEnumerator1
    Shape Primitive=::NewEnumerator4
    UseOwnerScale=true
   )
   &AddVelocity(
    Cone Angle=30
    Cone Axis=<0 1 0>
    Module.UseConeFalloff=false
    Module.UseVelDistribution=false
    Override Seed=false
    Rotation Coordinate Space=::Simulation
    Velocity Mode=::NewEnumerator2
    Velocity Speed=250
   )
   &SkeletalMeshLocation(
    Bone / Socket Index=$Particles.WheelID
    Bone / Socket Sampling Mode=::NewEnumerator5
    Interpolate Mesh Sampling=false
    MaskSpawnGroup=false
    Mesh Sampling Type=::NewEnumerator6
    Orientation Sampling=::NewEnumerator1
    Position Sampling=::NewEnumerator2
    Skeletal Mesh=$System.SkeletalMesh
   ) {module=SkeletalMeshLocation001}
   &VC_WheelXform(
    SKM Position:NiagaraPosition=$Particles.SkeletalMeshLocation001.SampledPosition
    SKM Rotation:Quat4f=$Particles.SkeletalMeshLocation001.SampledRotation
    SKM Velocity:Vector3f=$Particles.SkeletalMeshLocation001.SampledVelocity
    Write Attributes=true
   ) {module=VC_WheelXform001}
  ]
  particleUpdateScript [
   &ParticleState()
   &ScaleColor(
    Scale Alpha=&FloatFromCurve(FloatCurve=DataInterface() {propertyValues {Curve {keys [<at 0.75 1>,<at 1 0>]}}})
    ScaleA=true
    ScaleRGB=false
   )
   &Drag(Use Linear Drag=true, Use Rotational Drag=true)
   &SpriteFacingAndAlignment(Sprite Alignment=$Particles.Velocity, WriteAlignment=true, WriteFacing=false)
   &InheritSourceMovement(
    Applied Position Scale=&Multiply_VectorByFloat(
     Vector=<-1 -1 -1>
     Float=&FloatFromCurve(FloatCurve=DataInterface() {propertyValues {Curve {
         defaultValue=1
         keys [{interpMode=RCIM_Cubic, tangentMode=RCTM_User, time=0.25, value=1}, {interpMode=RCIM_Cubic, tangentMode=RCTM_User, time=1, value=0.8}]
        }}})
    )
    ApplyForce=false
    ApplyPosition=true
    ApplyPositionFalloff=false
   ) {enabled=false}
   &SolveForcesAndVelocity(Clamp Velocity=false, Limit Acceleration=false)
   &SkeletalMeshLocation(
    Bone / Socket Index=$Particles.WheelID
    Bone / Socket Sampling Mode=::NewEnumerator5
    MaskSpawnGroup=false
    Mesh Sampling Type=::NewEnumerator6
    Orientation Sampling=::NewEnumerator1
    Position Sampling=::NewEnumerator2
    Skeletal Mesh=$System.SkeletalMesh
   ) {enabled=false}
   &VC_WheelXform(
    SKM Position:NiagaraPosition=$Particles.SkeletalMeshLocation.SampledPosition
    SKM Rotation:Quat4f=$Particles.SkeletalMeshLocation.SampledRotation
    SKM Velocity:Vector3f=$Particles.SkeletalMeshLocation.SampledVelocity
   ) {enabled=false}
  ]
  rendererClasses [&NiagaraSpriteRendererProperties, &NiagaraMeshRendererProperties]
  renderers [
   {rendererClass=&NiagaraSpriteRendererProperties}
   {rendererClass=&NiagaraMeshRendererProperties, rendererIndex=1}
   {rendererClass=&NiagaraSpriteRendererProperties, rendererIndex=2}
  ]
 }
}
summary {
 systemName=NS_WheelFX_Example
 emitters [
  {emitterName=WheelFX_Mesh, rendererClasses [&NiagaraMeshRendererProperties, &NiagaraSpriteRendererProperties]}
  {
   emitterName=WheelFX_Sprites
   rendererClasses [&NiagaraMeshRendererProperties, &NiagaraSpriteRendererProperties]
  }
  {
   bEnabled=false
   emitterName=WheelFX_SkeletalMesh
   rendererClasses [
    &NiagaraComponentRendererProperties
    &NiagaraMeshRendererProperties
    &NiagaraSpriteRendererProperties
   ]
  }
  {
   emitterName=WheelFX_WorldSpace
   rendererClasses [&NiagaraSpriteRendererProperties, &NiagaraMeshRendererProperties]
  }
  {
   emitterName=WheelFX_WheelWrap
   rendererClasses [&NiagaraMeshRendererProperties, &NiagaraSpriteRendererProperties]
  }
 ]
 userVariables [
  User.bIsFrontendPreview {type=NiagaraBool}
  User.bIsPainted {type=NiagaraBool}
  User.ColorVariantID {type=NiagaraInt32}
  User.PaintedColor {type=LinearColor, description="", defaultValue {struct=LinearColor, value=<rgba 1 1 1 1>}}
  User.PaintedScalar {type=NiagaraFloat, defaultValue=1, description=""}
  User.VehicleFlags {type=NiagaraInt32, defaultValue=806289408}
  User.Wheel_Radius_Array {
   type=NiagaraDataInterfaceArrayFloat
   defaultValue {
    struct=NiagaraExt_VariableValue_DataInterface
    value {
     dataInterface=&self:NiagaraDataInterfaceArrayFloat_1
     dataInterfaceClass=&NiagaraDataInterfaceArrayFloat
    }
   }
  }
  User.Wheel_Width_Array {
   type=NiagaraDataInterfaceArrayFloat
   defaultValue {
    struct=NiagaraExt_VariableValue_DataInterface
    value {
     dataInterface=&self:NiagaraDataInterfaceArrayFloat_0
     dataInterfaceClass=&NiagaraDataInterfaceArrayFloat
    }
   }
  }
  User.WheelTouchingArray {
   type=NiagaraDataInterfaceArrayFloat
   defaultValue {
    struct=NiagaraExt_VariableValue_DataInterface
    value {
     dataInterface=&self:NiagaraDataInterfaceArrayFloat_2
     dataInterfaceClass=&NiagaraDataInterfaceArrayFloat
    }
   }
  }
 ]
}
