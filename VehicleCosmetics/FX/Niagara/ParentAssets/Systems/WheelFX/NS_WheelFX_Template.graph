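# Template Niagara system for vehicle cosmetic wheel FX, built around one emitter, WheelFX_Basic, with mesh and sprite renderers.
# It bursts one particle per wheel, using the partition index as WheelID for per-wheel setup and painted-variant overrides.
# Each update samples the skeletal mesh bone or socket by WheelID and feeds position, rotation and velocity into VC_WheelXform.
# Spawn rate, ring location and cone velocity modules are present but disabled.
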NiagaraSystem /VehicleCosmetics/FX/Niagara/ParentAssets/Systems/WheelFX/NS_WheelFX_Template
dependencies {
 usedDataInterfaces [&NiagaraDataInterfaceSkeletalMesh]
 usedDynamicInputs []
 usedModules [
  &self:SystemSpawnScript.SystemScriptSource.SystemScriptGraph.NiagaraNodeAssignment_0.SetVariables_#1
  &FN_UnpackVehicleFlags
  /VehicleCosmetics/FX/Niagara/Modules/WheelFX/VC_WheelFX_SystemSetup
  /VehicleCosmetics/FX/Niagara/Modules/VC_SetVehicleSpeedAttribs
  &SystemState
 ]
}
emitters {WheelFX_Basic {
  emitterSpawnScript [/VehicleCosmetics/FX/Niagara/Modules/WheelFX/VC_WheelFX_EmitterSetup()]
  emitterUpdateScript [
   &EmitterState(Life Cycle Mode=::NewEnumerator1, Loop Behavior=::NewEnumerator1, Loop Duration Mode=::NewEnumerator1, UseLoopDelay=false)
   &SpawnBurst_Instantaneous(Spawn Count=$Emitter.NumWheels, Use Spawn Probability=false, UseLoopCountLimit=false)
   &SpawnRate(SpawnRate=15, Use Spawn Probability=false) {enabled=false}
  ]
  particleSpawnScript [
   &PartitionParticles(Number Of Partitions=$Emitter.NumWheels)
   /VehicleCosmetics/FX/Niagara/Modules/WheelFX/VC_WheelFX_ParticleSetup(WheelID:NiagaraInt32=$Particles.PartitionParticles.Partition)
   &VC_PaintedVariantsOverride()
   &InitializeParticle(Color Mode=::NewEnumerator0, Position Mode=::NewEnumerator1, UsePositionOffset=false, Write Material Random=true)
   &ShapeLocation(
    MaskSpawnGroup=false
    Override Seed=false
    Ring Radius=&Multiply_Float(A=$Particles.WheelRadius, B=0.5)
    Rotation Angle=90
    Rotation Angle Type:ENiagara_AngleInput=::NewEnumerator0
    Rotation Axis=<0 1 0>
    Rotation Mode=::NewEnumerator1
    Shape Primitive=::NewEnumerator4
    UseOwnerScale=true
   ) {enabled=false}
   &AddVelocity(
    Cone Angle=30
    Cone Axis=<0 1 0>
    Module.UseConeFalloff=false
    Module.UseVelDistribution=false
    Override Seed=false
    Rotation Coordinate Space=::Simulation
    Velocity Mode=::NewEnumerator2
    Velocity Speed=150
   ) {enabled=false}
  ]
  particleUpdateScript [
   &ParticleState(Kill Particles When Lifetime Has Elapsed=false, Let Infinitely Lived Particles Die When Emitter Deactivates=true)
   &SolveForcesAndVelocity(Clamp Velocity=false, Limit Acceleration=false)
   &SkeletalMeshLocation(
    Bone / Socket Index=$Particles.WheelID
    Bone / Socket Sampling Mode=::NewEnumerator5
    MaskSpawnGroup=false
    Mesh Sampling Type=::NewEnumerator6
    Orientation Sampling=::NewEnumerator1
    Position Sampling=::NewEnumerator2
    Skeletal Mesh=$System.SkeletalMesh
   )
   /VehicleCosmetics/FX/Niagara/Modules/WheelFX/VC_WheelXform(
    SKM Position:NiagaraPosition=$Particles.SkeletalMeshLocation.SampledPosition
    SKM Rotation:Quat4f=$Particles.SkeletalMeshLocation.SampledRotation
    SKM Velocity:Vector3f=$Particles.SkeletalMeshLocation.SampledVelocity
   )
  ]
  rendererClasses [&NiagaraMeshRendererProperties, &NiagaraSpriteRendererProperties]
  renderers [
   {rendererClass=&NiagaraMeshRendererProperties, rendererIndex=0}
   {rendererClass=&NiagaraSpriteRendererProperties, rendererIndex=1}
  ]
 }}
summary {
 systemName=NS_WheelFX_Template
 emitters [{emitterName=WheelFX_Basic, rendererClasses [&NiagaraMeshRendererProperties, &NiagaraSpriteRendererProperties]}]
 userVariables [
  User.bIsFrontendPreview {type=NiagaraBool}
  User.bIsPainted {type=NiagaraBool}
  User.ColorVariantID {type=NiagaraInt32}
  User.PaintedColor {type=LinearColor, description="", defaultValue {struct=LinearColor, value=<rgba 1 1 1 1>}}
  User.PaintedScalar {type=NiagaraFloat, defaultValue=1, description=""}
  User.VehicleFlags {type=NiagaraInt32, defaultValue=806289408}
  User.Wheel_Radius_Array {
   type=NiagaraDataInterfaceArrayFloat
   defaultValue {
    struct=NiagaraExt_VariableValue_DataInterface
    value {
     dataInterface=&self:NiagaraDataInterfaceArrayFloat_1
     dataInterfaceClass=&NiagaraDataInterfaceArrayFloat
    }
   }
  }
  User.Wheel_Width_Array {
   type=NiagaraDataInterfaceArrayFloat
   defaultValue {
    struct=NiagaraExt_VariableValue_DataInterface
    value {
     dataInterface=&self:NiagaraDataInterfaceArrayFloat_0
     dataInterfaceClass=&NiagaraDataInterfaceArrayFloat
    }
   }
  }
  User.WheelTouchingArray {
   type=NiagaraDataInterfaceArrayFloat
   defaultValue {
    struct=NiagaraExt_VariableValue_DataInterface
    value {
     dataInterface=&self:NiagaraDataInterfaceArrayFloat_2
     dataInterfaceClass=&NiagaraDataInterfaceArrayFloat
    }
   }
  }
 ]
}
